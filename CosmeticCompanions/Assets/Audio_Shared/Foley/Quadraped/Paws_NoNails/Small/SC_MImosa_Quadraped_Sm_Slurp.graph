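# Sound cue for the slurp foley of small no-nail quadruped companion paws, built from Large-size slurp waves.
# A quality-level node sends its first two levels to a random pick of four waves and its last two to a pick of two.
# A modulator above it varies pitch and volume.
# Companion foley attenuation, the pets sound class and shared quadruped footstep concurrency govern playback.
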SoundCue /CosmeticCompanions/Assets/Audio_Shared/Foley/Quadraped/Paws_NoNails/Small/SC_MImosa_Quadraped_Sm_Slurp
root=Modulator_0
nodes [
 {id=Modulator_0, children [QualityLevel_0], properties {pitchMax=1.1, pitchMin=0.9, volumeMax=1, volumeMin=0.6}}
 {id=QualityLevel_0, children [Random_0,Random_0,Random_1,Random_1]}
 {id=Random_0, children [WavePlayer_0,WavePlayer_1,WavePlayer_2,WavePlayer_3], properties {weights [1,1,1,1]}}
 {id=Random_1, children [WavePlayer_4,WavePlayer_5], properties {weights [1,1]}}
 {
  id=WavePlayer_0
  properties {soundWaveAssetPtr=/CosmeticCompanions/Assets/Audio_Shared/Foley/Quadraped/Paws_NoNails/Large/Wavs/Mimosa_Paws_Lrg_Fol_Slurp_04}
 }
 {
  id=WavePlayer_1
  properties {soundWaveAssetPtr=/CosmeticCompanions/Assets/Audio_Shared/Foley/Quadraped/Paws_NoNails/Large/Wavs/Mimosa_Paws_Lrg_Fol_Slurp_03}
 }
 {
  id=WavePlayer_2
  properties {soundWaveAssetPtr=/CosmeticCompanions/Assets/Audio_Shared/Foley/Quadraped/Paws_NoNails/Large/Wavs/Mimosa_Paws_Lrg_Fol_Slurp_02}
 }
 {
  id=WavePlayer_3
  properties {soundWaveAssetPtr=/CosmeticCompanions/Assets/Audio_Shared/Foley/Quadraped/Paws_NoNails/Large/Wavs/Mimosa_Paws_Lrg_Fol_Slurp_01}
 }
 {
  id=WavePlayer_4
  properties {soundWaveAssetPtr=&Mimosa_Paws_Lrg_Fol_Slurp_01}
 }
 {
  id=WavePlayer_5
  properties {soundWaveAssetPtr=&Mimosa_Paws_Lrg_Fol_Slurp_02}
 }
]
properties {
 attenuationSettings=/Game/Sounds/Attenuation/AS_Companion_Foley
 maxDistance=1200
 priority=0
 soundClassObject=&Fort_Creatures_Pets
 volumeMultiplier=1.8
 concurrencySet [/CosmeticCompanions/Assets/Audio_Shared/Foley/Quadraped/CONC_Mimosa_Quadraped_Footstep]
}
